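; Camera modifier for the Beast Girl Storm Claws dive that offsets the camera and FOV.
; It blends in a view-rotation-relative location offset and an added FOV offset, both weighted by a curve sampled at the modifier's alpha, leaving rotation unchanged.
; The target camera rotation override points straight down by rotating the view target's forward vector 90 degrees around its right vector.
; The relative pitch frustum override simply returns the parent's value.
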(fn BlueprintModifyCamera
  (params (DeltaTime :type Float%20%28single-precision%29) (ViewLocation :type Vector :default "0, 0, 0") (ViewRotation :type Rotator :default "0, 0, 0") (FOV :type Float%20%28single-precision%29))
  (|CallParentFunction DeltaTime ViewLocation ViewRotation FOV)
  (return (Math|Vector|Lerp(Vector) ViewLocation (+ ViewLocation (Math|Vector|RotateVector (Variables|Default|GetCameraOffset) ViewRotation)) (Math|Curves|GetFloatValue (Variables|Default|GetCurve) (Variables|CameraModifier|GetAlpha))) ViewRotation (+ FOV (Math|Float|Lerp 0.0 (Variables|Default|GetFOVOffset) (Math|Curves|GetFloatValue (Variables|Default|GetCurve) (Variables|CameraModifier|GetAlpha))))))

(fn GetRelativePitchFrustum
  (bind _outrelativepitchfrustum (|CallParentFunction))
  (return _outrelativepitchfrustum))

(fn GetTargetCameraRotation
  (return (Math|Rotator|MakeRotfromX (Math|Vector|RotateVectorAroundAxis (Transformation|GetActorForwardVector (CameraModifier|GetViewTarget)) 9e+01 (Transformation|GetActorRightVector (CameraModifier|GetViewTarget))))))
